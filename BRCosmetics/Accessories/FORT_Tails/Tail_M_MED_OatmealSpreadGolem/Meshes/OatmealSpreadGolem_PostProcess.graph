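# Post-process AnimBlueprint for the OatmealSpreadGolem tail cosmetic, adding pose correction and physics sway.
# Additive bone offsets shape the input pose, cached into a world-space chain of per-bone Anim Dynamics with angular springs and custom gravity.
# A second cache feeds final bone-space rotations on the mid-to-outer tail.
# Skydiving and backbling events set AnimDyn Alpha, though no visible pin reads it; the TailUp branch is unwired.
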AnimBlueprint /BRCosmetics/Accessories/FORT_Tails/Tail_M_MED_OatmealSpreadGolem/Meshes/OatmealSpreadGolem_PostProcess
target_skeleton=/Game/Accessories/FORT_Tails/Common/Fortnite_Base_Tail_Skeleton
graphs [
 AnimGraph {nodes [
   n1 = "Transform (Modify) Bone - Bone: C_Tail_A_4_Jnt"(
    ComponentPose=Local To Component(LocalPose=Use cached pose 'DynamicsCache'())
    Alpha=1
    Translation="0,0,0"
    Rotation="20.000,-0.001,0.000"
    Scale="1,1,1"
   ) {
    rotationMode=BMM_Additive
    rotationSpace=BCS_BoneSpace
    scaleMode=BMM_Replace
    scaleSpace=BCS_BoneSpace
    translationMode=BMM_Additive
    translationSpace=BCS_BoneSpace
    alphaBoolBlend {blendOption=Linear}
    alphaScaleBiasClamp {bInterpResult=true, interpSpeedDecreasing=5}
    boneToModify {boneName=C_Tail_A_4_Jnt}
    rotation=<rot 20 -0.001 0>
   }
   n2 = "Transform (Modify) Bone - Bone: C_Tail_A_5_Jnt"(ComponentPose=@n1, Alpha=1, Translation="0,0,0", Rotation="19.872,0.242,-6.881", Scale="1,1,1") {
    rotationMode=BMM_Additive
    rotationSpace=BCS_BoneSpace
    scaleMode=BMM_Replace
    scaleSpace=BCS_BoneSpace
    translationMode=BMM_Additive
    translationSpace=BCS_BoneSpace
    alphaBoolBlend {blendOption=Linear}
    alphaScaleBiasClamp {bInterpResult=true, interpSpeedDecreasing=5}
    boneToModify {boneName=C_Tail_A_5_Jnt}
    rotation=<rot 19.872 0.242 -6.881>
   }
   n3 = "Transform (Modify) Bone - Bone: C_Tail_A_6_Jnt"(ComponentPose=@n2, Alpha=1, Translation="0,0,0", Rotation="29.805,-0.328,-2.214", Scale="1,1,1") {
    rotationMode=BMM_Additive
    rotationSpace=BCS_BoneSpace
    scaleMode=BMM_Replace
    scaleSpace=BCS_BoneSpace
    translationMode=BMM_Additive
    translationSpace=BCS_BoneSpace
    alphaBoolBlend {blendOption=Linear}
    alphaScaleBiasClamp {bInterpResult=true, interpSpeedDecreasing=5}
    boneToModify {boneName=C_Tail_A_6_Jnt}
    rotation=<rot 29.8049 -0.328 -2.214>
   }
   n4 = "Transform (Modify) Bone - Bone: C_Tail_A_7_Jnt"(ComponentPose=@n3, Alpha=1, Translation="0,0,0", Rotation="19.542,-1.136,-3.500", Scale="1,1,1") {
    rotationMode=BMM_Additive
    rotationSpace=BCS_BoneSpace
    scaleMode=BMM_Replace
    scaleSpace=BCS_BoneSpace
    translationMode=BMM_Additive
    translationSpace=BCS_BoneSpace
    alphaBoolBlend {blendOption=Linear}
    alphaScaleBiasClamp {bInterpResult=true, interpSpeedDecreasing=5}
    boneToModify {boneName=C_Tail_A_7_Jnt}
    rotation=<rot 19.542 -1.136 -3.5>
   }
   n5 = "Transform (Modify) Bone - Bone: C_Tail_A_8_Jnt"(ComponentPose=@n4, Alpha=1, Translation="0,0,0", Rotation="-0.346,-0.687,0.700", Scale="1,1,1") {
    rotationMode=BMM_Additive
    rotationSpace=BCS_BoneSpace
    scaleMode=BMM_Replace
    scaleSpace=BCS_BoneSpace
    translationMode=BMM_Additive
    translationSpace=BCS_BoneSpace
    alphaBoolBlend {blendOption=Linear}
    alphaScaleBiasClamp {bInterpResult=true, interpSpeedDecreasing=5}
    boneToModify {boneName=C_Tail_A_8_Jnt}
    rotation=<rot -0.346 -0.687 0.7>
   }
   Output Pose(Result=Component To Local(ComponentPose=@n5))
   n6 = "Transform (Modify) Bone - Bone: C_Root_Main_Root_Jnt"(ComponentPose=Local To Component(LocalPose=Input Pose()), Alpha=1, Translation="0,0,-1", Rotation="0,0,0", Scale="1,1,1") {
    rotationMode=BMM_Additive
    translationMode=BMM_Additive
    alphaBoolBlend {blendOption=Linear}
    boneToModify {boneName=C_Root_Main_Root_Jnt}
    translation {z=-1}
   }
   n7 = "Transform (Modify) Bone - Bone: C_Tail_A_Base_Jnt"(ComponentPose=@n6, Alpha=1, Translation="-1.9,0,0", Rotation="0,0,0", Scale="1,1,1") {
    rotationMode=BMM_Additive
    translationMode=BMM_Additive
    alphaBoolBlend {blendOption=Linear}
    boneToModify {boneName=C_Tail_A_Base_Jnt}
    translation {x=-1.9}
   }
   n8 = "Transform (Modify) Bone - Bone: C_Tail_A_Base_Jnt"(ComponentPose=@n7, Alpha=1, Translation="0,0,0", Rotation="0,0,0", Scale="1,1,1") {
    rotationMode=BMM_Additive
    translationMode=BMM_Additive
    alphaBoolBlend {blendOption=Linear}
    alphaScaleBiasClamp {bInterpResult=true, interpSpeedDecreasing=1, interpSpeedIncreasing=1}
    boneToModify {boneName=C_Tail_A_Base_Jnt}
   }
   n9 = "Transform (Modify) Bone - Bone: C_Tail_A_3_Jnt"(ComponentPose=@n8, Alpha=1, Translation="0,0,0", Rotation="0.000,0.000,10.000", Scale="1,1,1") {
    rotationMode=BMM_Additive
    translationMode=BMM_Additive
    alphaBoolBlend {blendOption=Linear}
    alphaScaleBiasClamp {bInterpResult=true, interpSpeedDecreasing=1, interpSpeedIncreasing=1}
    boneToModify {boneName=C_Tail_A_3_Jnt}
    rotation {roll=10}
   }
   n10 = "Transform (Modify) Bone - Bone: C_Tail_A_4_Jnt"(ComponentPose=@n9, Alpha=1, Translation="0,0,0", Rotation="0,0,0", Scale="1,1,1") {
    rotationMode=BMM_Additive
    translationMode=BMM_Additive
    alphaBoolBlend {blendOption=Linear}
    alphaScaleBiasClamp {bInterpResult=true, interpSpeedDecreasing=1, interpSpeedIncreasing=1}
    boneToModify {boneName=C_Tail_A_4_Jnt}
   }
   n11 = "Transform (Modify) Bone - Bone: C_Tail_A_5_Jnt"(ComponentPose=@n10, Alpha=1, Translation="0,0,0", Rotation="0.000,0.000,20.000", Scale="1,1,1") {
    rotationMode=BMM_Additive
    translationMode=BMM_Additive
    alphaBoolBlend {blendOption=Linear}
    alphaScaleBiasClamp {bInterpResult=true, interpSpeedDecreasing=1, interpSpeedIncreasing=1}
    boneToModify {boneName=C_Tail_A_5_Jnt}
    rotation {roll=20}
   }
   n12 = "Transform (Modify) Bone - Bone: C_Tail_A_6_Jnt"(ComponentPose=@n11, Alpha=1, Translation="0,0,0", Rotation="0,0,0", Scale="1,1,1") {
    rotationMode=BMM_Additive
    translationMode=BMM_Additive
    alphaBoolBlend {blendOption=Linear}
    alphaScaleBiasClamp {bInterpResult=true, interpSpeedDecreasing=1, interpSpeedIncreasing=1}
    boneToModify {boneName=C_Tail_A_6_Jnt}
   }
   n13 = "Transform (Modify) Bone - Bone: C_Tail_A_7_Jnt"(ComponentPose=@n12, Alpha=1, Translation="0,0,0", Rotation="0.000,0.000,20.000", Scale="1,1,1") {
    rotationMode=BMM_Additive
    translationMode=BMM_Additive
    alphaBoolBlend {blendOption=Linear}
    alphaScaleBiasClamp {bInterpResult=true, interpSpeedDecreasing=1, interpSpeedIncreasing=1}
    boneToModify {boneName=C_Tail_A_7_Jnt}
    rotation {roll=20}
   }
   n14 = "Transform (Modify) Bone - Bone: C_Tail_A_8_Jnt"(ComponentPose=@n13, Alpha=1, Translation="0,0,0", Rotation="0.000,0.000,10.000", Scale="1,1,1") {
    rotationMode=BMM_Additive
    translationMode=BMM_Additive
    alphaBoolBlend {blendOption=Linear}
    alphaScaleBiasClamp {bInterpResult=true, interpSpeedDecreasing=1, interpSpeedIncreasing=1}
    boneToModify {boneName=C_Tail_A_8_Jnt}
    rotation {roll=10}
   }
   n15 = "Transform (Modify) Bone - Bone: C_Tail_A_9_Jnt"(ComponentPose=@n14, Alpha=1, Translation="0,0,0", Rotation="0,0,0", Scale="1,1,1") {
    rotationMode=BMM_Additive
    translationMode=BMM_Additive
    alphaBoolBlend {blendOption=Linear}
    alphaScaleBiasClamp {bInterpResult=true, interpSpeedDecreasing=1, interpSpeedIncreasing=1}
    boneToModify {boneName=C_Tail_A_9_Jnt}
   }
   Save cached pose 'SetupCache'(Pose=Component To Local(ComponentPose=@n15)) {cacheName=SetupCache}
   n16 = Anim Dynamics - Bone: C_Tail_A_1_Jnt!(ComponentPose=Local To Component(LocalPose=Use cached pose 'SetupCache'()), Alpha=0.5) -> (Pose) {
    angularDampingOverride=0.999
    angularSpringConstant=900
    bAngularSpring=true
    bGravityOverrideInSimSpace=true
    bOverrideAngularDamping=true
    bUseGravityOverride=true
    gravityScale=0.35
    simulationSpace=World
    alphaBoolBlend {blendOption=Linear}
    alphaScaleBiasClamp {bInterpResult=true}
    boundBone {boneName=C_Tail_A_1_Jnt}
    componentAppliedLinearAccClamp=<500 500 500>
    componentLinearAccScale=<1 0.5 0.5>
    componentLinearVelScale=<1 0.5 0.5>
    gravityOverride {z=-1300}
    physicsBodyDefinitions [{
      boundBone {boneName=C_Tail_A_1_Jnt}
      boxExtents=<12 12 12>
      constraintSetup {angularLimitsMax=<0 3 25>, angularLimitsMin=<0 -25 -25>, angularTarget=<1 0 0>}
      localJointOffset=<0.5 0 0>
     }]
   }
   n17 = Anim Dynamics - Bone: C_Tail_A_2_Jnt!(ComponentPose=@n16, Alpha=0.5) -> (Pose) {
    angularDampingOverride=0.999
    angularSpringConstant=900
    bAngularSpring=true
    bGravityOverrideInSimSpace=true
    bOverrideAngularDamping=true
    bUseGravityOverride=true
    gravityScale=0.35
    simulationSpace=World
    alphaBoolBlend {blendOption=Linear}
    alphaScaleBiasClamp {bInterpResult=true}
    boundBone {boneName=C_Tail_A_2_Jnt}
    componentAppliedLinearAccClamp=<500 500 500>
    componentLinearAccScale=<1 0.5 0.5>
    componentLinearVelScale=<1 0.5 0.5>
    gravityOverride {z=-1300}
    physicsBodyDefinitions [{
      boundBone {boneName=C_Tail_A_2_Jnt}
      boxExtents=<12 12 12>
      constraintSetup {angularLimitsMax=<0 3 25>, angularLimitsMin=<0 -25 -25>, angularTarget=<1 0 0>}
      localJointOffset=<0.5 0 0>
     }]
   }
   n18 = Anim Dynamics - Bone: C_Tail_A_3_Jnt!(ComponentPose=@n17, Alpha=0.5) -> (Pose) {
    angularDampingOverride=0.999
    angularSpringConstant=900
    bAngularSpring=true
    bGravityOverrideInSimSpace=true
    bOverrideAngularDamping=true
    bUseGravityOverride=true
    gravityScale=0.35
    simulationSpace=World
    alphaBoolBlend {blendOption=Linear}
    alphaScaleBiasClamp {bInterpResult=true}
    boundBone {boneName=C_Tail_A_3_Jnt}
    componentAppliedLinearAccClamp=<500 500 500>
    componentLinearAccScale=<1 0.5 0.5>
    componentLinearVelScale=<1 0.5 0.5>
    gravityOverride {z=-1300}
    physicsBodyDefinitions [{
      boundBone {boneName=C_Tail_A_3_Jnt}
      boxExtents=<12 12 12>
      constraintSetup {angularLimitsMax=<0 3 25>, angularLimitsMin=<0 -25 -25>, angularTarget=<1 0 0>}
      localJointOffset=<0.5 0 0>
     }]
   }
   n19 = Anim Dynamics - Bone: C_Tail_A_4_Jnt!(ComponentPose=@n18, Alpha=0.5) -> (Pose) {
    angularDampingOverride=0.999
    angularSpringConstant=900
    bAngularSpring=true
    bGravityOverrideInSimSpace=true
    bOverrideAngularDamping=true
    bUseGravityOverride=true
    gravityScale=0.35
    simulationSpace=World
    alphaBoolBlend {blendOption=Linear}
    alphaScaleBiasClamp {bInterpResult=true}
    boundBone {boneName=C_Tail_A_4_Jnt}
    componentAppliedLinearAccClamp=<500 500 500>
    componentLinearAccScale=<1 0.5 0.5>
    componentLinearVelScale=<1 0.5 0.5>
    gravityOverride {z=-1300}
    physicsBodyDefinitions [{
      boundBone {boneName=C_Tail_A_4_Jnt}
      boxExtents=<12 12 12>
      constraintSetup {angularLimitsMax=<0 3 25>, angularLimitsMin=<0 -35 -25>, angularTarget=<1 0 0>}
      localJointOffset=<0.5 0 0>
     }]
   }
   n20 = Anim Dynamics - Bone: C_Tail_A_5_Jnt!(ComponentPose=@n19, Alpha=0.5) -> (Pose) {
    angularDampingOverride=0.999
    angularSpringConstant=900
    bAngularSpring=true
    bGravityOverrideInSimSpace=true
    bOverrideAngularDamping=true
    bUseGravityOverride=true
    gravityScale=0.35
    simulationSpace=World
    alphaBoolBlend {blendOption=Linear}
    alphaScaleBiasClamp {bInterpResult=true}
    boundBone {boneName=C_Tail_A_5_Jnt}
    componentAppliedLinearAccClamp=<500 500 500>
    componentLinearAccScale=<1 0.5 0.5>
    componentLinearVelScale=<1 0.5 0.5>
    gravityOverride {z=-1300}
    physicsBodyDefinitions [{
      boundBone {boneName=C_Tail_A_5_Jnt}
      boxExtents=<12 12 12>
      constraintSetup {angularLimitsMax=<0 3 25>, angularLimitsMin=<0 -35 -25>, angularTarget=<1 0 0>}
      localJointOffset=<0.5 0 0>
     }]
   }
   n21 = Anim Dynamics - Bone: C_Tail_A_6_Jnt!(ComponentPose=@n20, Alpha=0.5) -> (Pose) {
    angularDampingOverride=0.999
    angularSpringConstant=900
    bAngularSpring=true
    bGravityOverrideInSimSpace=true
    bOverrideAngularDamping=true
    bUseGravityOverride=true
    gravityScale=0.35
    simulationSpace=World
    alphaBoolBlend {blendOption=Linear}
    alphaScaleBiasClamp {bInterpResult=true}
    boundBone {boneName=C_Tail_A_6_Jnt}
    componentAppliedLinearAccClamp=<500 500 500>
    componentLinearAccScale=<1 0.5 0.5>
    componentLinearVelScale=<1 0.5 0.5>
    gravityOverride {z=-1300}
    physicsBodyDefinitions [{
      boundBone {boneName=C_Tail_A_6_Jnt}
      boxExtents=<14 14 14>
      constraintSetup {angularLimitsMax=<0 3 25>, angularLimitsMin=<0 -35 -25>, angularTarget=<1 0 0>}
      localJointOffset=<0.5 0 0>
     }]
   }
   n22 = Anim Dynamics - Bone: C_Tail_A_7_Jnt!(ComponentPose=@n21, Alpha=0.5) -> (Pose) {
    angularDampingOverride=0.999
    angularSpringConstant=900
    bAngularSpring=true
    bGravityOverrideInSimSpace=true
    bOverrideAngularDamping=true
    bUseGravityOverride=true
    gravityScale=0.35
    simulationSpace=World
    alphaBoolBlend {blendOption=Linear}
    alphaScaleBiasClamp {bInterpResult=true}
    boundBone {boneName=C_Tail_A_7_Jnt}
    componentAppliedLinearAccClamp=<500 500 500>
    componentLinearAccScale=<1 0.5 0.5>
    componentLinearVelScale=<1 0.5 0.5>
    gravityOverride {z=-1300}
    physicsBodyDefinitions [{
      boundBone {boneName=C_Tail_A_7_Jnt}
      boxExtents=<15 15 15>
      constraintSetup {angularLimitsMax=<0 3 25>, angularLimitsMin=<0 -30 -25>, angularTarget=<1 0 0>}
      localJointOffset=<0.5 0 0>
     }]
   }
   n23 = Anim Dynamics - Bone: C_Tail_A_8_Jnt!(ComponentPose=@n22, Alpha=0.5) -> (Pose) {
    angularDampingOverride=0.999
    angularSpringConstant=900
    bAngularSpring=true
    bGravityOverrideInSimSpace=true
    bOverrideAngularDamping=true
    bUseGravityOverride=true
    gravityScale=0.35
    simulationSpace=World
    alphaBoolBlend {blendOption=Linear}
    alphaScaleBiasClamp {bInterpResult=true}
    boundBone {boneName=C_Tail_A_8_Jnt}
    componentAppliedLinearAccClamp=<500 500 500>
    componentLinearAccScale=<1 0.5 0.5>
    componentLinearVelScale=<1 0.5 0.5>
    gravityOverride {z=-1300}
    physicsBodyDefinitions [{
      boundBone {boneName=C_Tail_A_8_Jnt}
      boxExtents=<16 16 16>
      constraintSetup {angularLimitsMax=<0 3 25>, angularLimitsMin=<0 -25 -25>, angularTarget=<1 0 0>}
      localJointOffset=<0.5 0 0>
     }]
   }
   Save cached pose 'DynamicsCache'(Pose=Component To Local(ComponentPose=@n23)) {cacheName=DynamicsCache}
  ]}
 EventGraph {nodes [
   Event OnBeginSkydiving()
   > Set AnimDyn Alpha(AnimDyn Alpha:double=0) -> (Output_Get:double)
   Event OnEndSkydiving()
   > Set AnimDyn Alpha(AnimDyn Alpha=0.5)
   Event OnCharacterCustomizationLoaded()
   > n1 = Branch(Condition=CheckForBackbling())
   > Set AnimDyn Alpha(AnimDyn Alpha=0.1)
   Set AnimDyn Alpha(execute=@n1.else, AnimDyn Alpha=0.5)
   n2 = Branch()
   > Set TailUp(TailUp:double=0) -> (Output_Get:double)
   Set TailUp(execute=@n2.else, TailUp=1)
   TryGetPawnOwner()
   Event BlueprintUpdateAnimation()
  ]}
]
variables [
 AnimDyn Alpha {type=number, default=0.5}
 TailUp {type=number}
]
